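# Foley sound cue for the NeonCatSpy emote in Battle Royale.
# A Random node picks one of four equally weighted wave variations each time it plays.
# It uses the emote foley sound class and property sheet with Athena emote attenuation.
# Though it is foley, it references the emote music concurrency rule and sends to the music audio-analysis submix.
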SoundCue /BRCosmetics/Sounds/Emotes/NeonCatSpy/Athena_Emote_NeonCatSpy_Foley_Cue
root=Random_0
nodes [
 {id=Random_0, children [WavePlayer_0,WavePlayer_1,WavePlayer_2,WavePlayer_4], properties {weights [1,1,1,1]}}
 {id=WavePlayer_0, properties {soundWaveAssetPtr=&/Emote_NeonCatSpy_04}}
 {id=WavePlayer_1, properties {soundWaveAssetPtr=&/Emote_NeonCatSpy_01}}
 {id=WavePlayer_2, properties {soundWaveAssetPtr=&/Emote_NeonCatSpy_02}}
 {id=WavePlayer_4, properties {soundWaveAssetPtr=&/Emote_NeonCatSpy_03}}
]
properties {
 attenuationSettings=&Fort_Emotes_Athena_Att
 audioPropertiesSheet=/BRCosmetics/Sounds/Settings/PropertySheets/Emotes/APS_Cosmetic_Emote_Foley
 maxDistance=1600
 soundClassObject=/Game/Sounds/FortSoundClassesAndModes/Fort_Foley_Emote
 volumeMultiplier=1
 concurrencyOverrides {maxCount=3, resolutionRule=StopOldest}
 concurrencySet [/Game/Sounds/Concurrency_Rules/Athena_EmoteMusic_Conc]
 soundSubmixSends [{minSendLevel=0, soundSubmix=&Fortnite_AudioAnalysis_Music}]
}
